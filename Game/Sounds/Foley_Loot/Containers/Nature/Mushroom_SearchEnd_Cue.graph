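# Sound cue for the end of searching a mushroom container, pairing foley with a "no more loot" sting.
# A mixer layers the slightly lowered UI sting over one of two random mushroom search waves.
# The foley branch fades out through a volume envelope over about 1.4 seconds, then gets slight random pitch and volume modulation.
# It uses prop-search attenuation, the loot-collect sound class, and a concurrency cap that stops the farthest instance first.
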SoundCue /Game/Sounds/Foley_Loot/Containers/Nature/Mushroom_SearchEnd_Cue
root=Mixer_12
nodes [
 {
  id=Enveloper_21
  children [Random_10]
  properties {pitchCurve {editorCurveData {keys [{time=0}]}}, volumeCurve {editorCurveData {keys [<at 0 1>,<at 1.4041 -0.0146>]}}}
 }
 {id=Mixer_12, children [WavePlayer_43, Modulator_15], properties {inputVolume [0.8,1]}}
 {id=Modulator_15, children [Enveloper_21], properties {pitchMax=1.07, pitchMin=0.92, volumeMax=1.07, volumeMin=0.92}}
 {id=Random_10, children [WavePlayer_50, WavePlayer_51], properties {weights [1,1]}}
 {id=WavePlayer_43, properties {soundWaveAssetPtr=/Game/Sounds/Fort_UI_Sounds/Gameplay/Fort_UI_NoMoreLoot}}
 {id=WavePlayer_50, properties {soundWaveAssetPtr=&/Search_Mushrooms_01}}
 {id=WavePlayer_51, properties {soundWaveAssetPtr=&/Search_Mushrooms_02}}
]
properties {
 attenuationSettings=/Game/Sounds/Attenuation/Fort_Prop_Search_Att
 bOverrideConcurrency=true
 maxDistance=1300
 soundClassObject=&Fort_Loot_Collect
 volumeMultiplier=0.7
 concurrencyOverrides {maxCount=4, resolutionRule=StopFarthestThenPreventNew}
}
