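# Sound cue for searching a mushroom container, part of the nature loot foley set.
# A Random node picks between two mushroom search waves at equal weight, then a Modulator slightly varies volume and raises pitch by up to a small amount.
# It plays through the prop search attenuation in the loot collect sound class, with a concurrency override that caps simultaneous instances and stops the farthest first.
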SoundCue /Game/Sounds/Foley_Loot/Containers/Nature/Mushroom_Search_Cue
root=Modulator_15
nodes [
 {id=Modulator_15, children [Random_11], properties {pitchMax=1.07, pitchMin=1, volumeMax=1.07, volumeMin=0.92}}
 {id=Random_11, children [WavePlayer_52, WavePlayer_53], properties {weights [1,1]}}
 {id=WavePlayer_52, properties {soundWaveAssetPtr=&/Search_Mushrooms_01}}
 {id=WavePlayer_53, properties {soundWaveAssetPtr=&/Search_Mushrooms_02}}
]
properties {
 attenuationSettings=/Game/Sounds/Attenuation/Fort_Prop_Search_Att
 bOverrideConcurrency=true
 maxDistance=1300
 soundClassObject=&Fort_Loot_Collect
 volumeMultiplier=1.2
 concurrencyOverrides {maxCount=4, resolutionRule=StopFarthestThenPreventNew}
}
